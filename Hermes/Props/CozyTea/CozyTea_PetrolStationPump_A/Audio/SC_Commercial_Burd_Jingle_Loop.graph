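# Sound cue for a looping Burd commercial jingle on the CozyTea petrol station pump prop, classed as Athena ambient audio.
# A looping wave player feeds an Enveloper root whose pitch and volume curves each hold a single key.
# Playback is quiet, spatialized through a pump-specific attenuation asset, and overrides concurrency to cap simultaneous copies at three.
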SoundCue /Hermes/Props/CozyTea/CozyTea_PetrolStationPump_A/Audio/SC_Commercial_Burd_Jingle_Loop
root=Enveloper_0
nodes [
 {
  id=Enveloper_0
  children [WavePlayer_7]
  properties {pitchCurve {editorCurveData {keys [{time=0}]}}, volumeCurve {editorCurveData {keys [<at 0 1>]}}}
 }
 {
  id=WavePlayer_7
  properties {bLooping=true, soundWaveAssetPtr=&/Commercial_Burd_Jingle_Loop}
 }
]
properties {
 attenuationSettings=&/ATT_Commercial_Burd_Jingle_Pump
 bOverrideConcurrency=true
 maxDistance=350
 soundClassObject=&Fort_Ambient_Athena
 volumeMultiplier=0.2
 concurrencyOverrides {maxCount=3}
}
